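; Generic event spring-arm camera rig with player look control, distance changes, ADS zoom and timed focus on an actor.
; SpecialEventInputHelper look input drives clamped pivot rotation, and a timer starts a timeline to recenter it.
; Shared distance and rotation timelines route an enum through update and finish managers, chaining focus-in, hold and reset.
; ADS moves CameraAttach separately from arm length, and focus points only rotate the camera during ADS.
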(event EventBeginPlay)

(event Custom|ActivateCameraInput
  (params (EnableCameraInput :type Boolean))
  (bind _specialeventinputhelper (Variables|Default|GetSpecialEventInputHelper))
  (bind _specialeventinputhelper_2 (Variables|Default|GetSpecialEventInputHelper))
  ;[0 1 0.3885] Activate and Deactivate Camera Control
  (if EnableCameraInput
    (Components|Activation|Activate _specialeventinputhelper false)
    (Class|SpecialEventInputHelperComponent|PushInputComponent _specialeventinputhelper)
    (Variables|PivotRotation|SetAllowLookInput true)
    (else
      (Components|Activation|Deactivate _specialeventinputhelper_2)
      (Class|SpecialEventInputHelperComponent|PopInputComponent _specialeventinputhelper_2)
      (CallFunction|RecenterCameraPivot 0.0)
      (Variables|PivotRotation|SetAllowLookInput false))))

(event SpecialEventLookInputUpdated(SpecialEventInputHelper)
  (params (InputVector :type Vector%20%28by%20ref%29 :default "0, 0, 0"))
  ;[1 0 0.6024] Handle Rotation Input and Auto Return to Center
  (Variables|PivotRotation|SetCurrentInputVector InputVector))

(event EventTick
  (params (DeltaSeconds :type Float%20%28single-precision%29))
  ;[1 0 0.6024] Handle Rotation Input and Auto Return to Center
  (if (Variables|PivotRotation|GetAllowLookInput)
    (PivotRotation|UpdateCameraPivotRotation (Variables|PivotRotation|GetCurrentInputVector))
    (if (and (Variables|PivotRotation|GetPivotIsRotated) (not (Utilities|Time|IsValidTimerHandle (Variables|PivotRotation|GetRecenterCameraTimer))))
      (bind _returnvalue (Utilities|Time|SetTimerbyEvent (EventDispatchers|CreateEvent :!event Custom|RecenterCameraPivotOnTimer) (Variables|PivotRotation|GetTimeBeforeRecenter)))
      (Variables|PivotRotation|SetRecenterCameraTimer _returnvalue))))

(event Custom|RecenterCameraPivot
  (params (RecenterTime :type Float%20%28double-precision%29))
  ;[1 0 0.6024] Handle Rotation Input and Auto Return to Center
  ;[1 0.3764 0] If already recentering, don't.
  (if (and (Variables|PivotRotation|GetPivotIsRotated) (not (Components|Timeline|IsPlaying (Variables|BP_Generic_SpringArm|GetRecenterCameraPivotTimeline))))
    (if (> RecenterTime 0)
      (Components|Timeline|SetPlayRate (Variables|BP_Generic_SpringArm|GetRecenterCameraPivotTimeline) (/ 1.0 RecenterTime))
      (else
        (Components|Timeline|SetPlayRate (Variables|BP_Generic_SpringArm|GetRecenterCameraPivotTimeline) 1.0)))
    (bind (_direction _recenteralpha) (Timeline|RecenterCameraPivotTimeline PlayFromStart)
      (:Update
        (PivotRotation|RecenterCameraPivotRotation _recenteralpha))
      (:Finished
        (Variables|PivotRotation|SetCurrentCameraRotationZ 0.0)
        (Variables|PivotRotation|SetCurrentCameraRotationY 0.0)
        (Variables|PivotRotation|SetPivotIsRotated false)))))

(event Custom|Stop%20Recenter
  ;[1 0 0.6024] Handle Rotation Input and Auto Return to Center
  (Timeline|RecenterCameraPivotTimeline Stop))

(event Custom|RecenterCameraPivotOnTimer
  ;[1 0 0.6024] Handle Rotation Input and Auto Return to Center
  ;[1 0.3764 0] If already recentering, don't.
  (if (and (Variables|PivotRotation|GetPivotIsRotated) (not (Components|Timeline|IsPlaying (Variables|BP_Generic_SpringArm|GetRecenterCameraPivotTimeline))))
    (Components|Timeline|SetPlayRate (Variables|BP_Generic_SpringArm|GetRecenterCameraPivotTimeline) 1.0)
    (Timeline|RecenterCameraPivotTimeline PlayFromStart)))

(event EventEndPlay
  (params (EndPlayReason :type EEndPlayReason%20Enum :default "Destroyed"))
  ;[1 0.4858 0] EndPlay
  (CallFunction|ActivateCameraInput false))

(event Custom|StartCameraDistanceChangeTimeline
  (params (Timeline%20Length%20Seconds :type Float%20%28double-precision%29 :default 1.0) (Timeline%20Type :type Events_Generic_SpringArmMovements%20Enum :default "NewEnumerator0") (TargetDistance :type Float%20%28double-precision%29))
  ;[0.1606 0 1] Updates run through single timeline so that multiple changes don't overlap unintentionally. If you plan to overlap consider how you plan to call the finished timeline functionality.
  ;;[1 0.7096 0] Called to adjust camera attach positioning or spring arm length. Used for ADS or just general camera depth changes.
  (if (> Timeline%20Length%20Seconds 0)
    (Components|Timeline|SetPlayRate (Variables|BP_Generic_SpringArm|GetCameraDistanceChangeTimeline) (/ 1.0 Timeline%20Length%20Seconds))
    (else
      (Components|Timeline|SetPlayRate (Variables|BP_Generic_SpringArm|GetCameraDistanceChangeTimeline) 1.0)))
  (bind (_direction _recenteralpha) (Timeline|CameraDistanceChangeTimeline PlayFromStart)
    (:Update
      (CameraTimelineFunctions|CameraTimelineUpdateManager Timeline%20Type _recenteralpha TargetDistance))
    (:Finished
      (CameraTimelineFunctions|CameraTimelineFinishedManager Timeline%20Type 1.0 TargetDistance))))

(event Custom|StartCameraRotationChangeTimeline
  (params (TimelineLengthSeconds :type Float%20%28double-precision%29 :default 1.0) (Timeline%20Type :type Events_Generic_SpringArmMovements%20Enum :default "NewEnumerator0") (Target%20Rotation :type Rotator))
  ;[0.1606 0 1] Updates run through single timeline so that multiple changes don't overlap unintentionally. If you plan to overlap consider how you plan to call the finished timeline functionality.
  ;;[1 0.5488 0] Called when Focus Point is active to adjust the CameraAttach Rotation
  (if (> TimelineLengthSeconds 0)
    (Components|Timeline|SetPlayRate (Variables|BP_Generic_SpringArm|GetCameraRotationChangeTimeline) (/ 1.0 TimelineLengthSeconds))
    (else
      (Components|Timeline|SetPlayRate (Variables|BP_Generic_SpringArm|GetCameraRotationChangeTimeline) 1.0)))
  (bind (_direction _recenteralpha) (Timeline|CameraRotationChangeTimeline PlayFromStart)
    (:Update
      (CameraTimelineFunctions|CameraTimelineUpdateManager Timeline%20Type _recenteralpha 0.0 Target%20Rotation))
    (:Finished
      (CameraTimelineFunctions|CameraTimelineFinishedManager Timeline%20Type 1.0 0.0 Target%20Rotation))))

(fn ActivateFocusPoint
  ;[0.7407 0 1] Called when we want the Camera to rotate to the focus point
  (if (Variables|CameraDistance|GetCameraModeADS)
    (Class|SpecialEventInputHelperComponent|PopInputComponent (Variables|Default|GetSpecialEventInputHelper))
    (CallFunction|StartCameraRotationChangeTimeline (Variables|FocusPoint|GetTime%20to%20Focus) "NewEnumerator3")))

(fn ADSEnabledUpdate
  (params (Target%20Camera%20Attach%20Location :type Vector) (Alpha :type Float%20%28double-precision%29) (Target%20Camera%20Distance :type Float%20%28double-precision%29))
  (CameraTimelineFunctions|UpdateDistanceTimeline|UpdateCameraDistance Alpha Target%20Camera%20Distance)
  (Transformation|SetRelativeLocation (Variables|Default|GetCameraAttach) (Math|Vector|Lerp(Vector) (Variables|/Script/Engine.SceneComponent|GetRelativeLocation (Variables|Default|GetCameraAttach)) Target%20Camera%20Attach%20Location Alpha) false false))

(fn CameraTimelineFinishedManager
  (params (Timeline%20Type :type Events_Generic_SpringArmMovements%20Enum :default "NewEnumerator0") (SecondTimelineLength :type Float%20%28double-precision%29 :default 1.0) (SecondTimelineTargetDistance :type Float%20%28double-precision%29) (SecondTimelineTargetRotation :type Rotator))
  (bind _focuspointtimer (Variables|FocusPoint|GetFocusPointTimer))
  (switch Utilities|FlowControl|Switch|SwitchonEvents_Generic_SpringArmMovements Timeline%20Type
    (:NewEnumerator0
      (CameraTimelineFunctions|FinishedDistanceTImeline|SetArmLengthPostChange true))
    (:NewEnumerator1
      _)
    (:NewEnumerator2
      _)
    (:NewEnumerator3
      (CallFunction|StartCameraRotationChangeTimeline (Utilities|Time|GetTimerRemainingTimebyHandle _focuspointtimer) "NewEnumerator4"))
    (:NewEnumerator4
      (Utilities|Time|ClearandInvalidateTimerbyHandle _focuspointtimer)
      (CallFunction|StartCameraRotationChangeTimeline (Variables|FocusPoint|GetFocusPointResetTime) "NewEnumerator5"))
    (:NewEnumerator5
      (Class|SpecialEventInputHelperComponent|PushInputComponent (Variables|Default|GetSpecialEventInputHelper))))
  (CameraTimelineFunctions|FinishedDistanceTImeline|SetArmLengthPostChange false))

(fn CameraTimelineUpdateManager
  (params (TimelineType :type Events_Generic_SpringArmMovements%20Enum :default "NewEnumerator0") (Alpha :type Float%20%28double-precision%29) (TargetDistance :type Float%20%28double-precision%29) (TargetRotation :type Rotator))
  (LocalVariables|SetInputAlpha Alpha)
  (LocalVariables|SetInputTargetDistance TargetDistance)
  (switch Utilities|FlowControl|Switch|SwitchonEvents_Generic_SpringArmMovements TimelineType
    (:NewEnumerator0
      ;[1 1 1] Sets spring arm to new length
      (CameraTimelineFunctions|UpdateDistanceTimeline|UpdateCameraDistance (LocalVariables|GetInputAlpha) (LocalVariables|GetInputTargetDistance)))
    (:NewEnumerator1
      ;[1 1 1] Move CameraAttach from SpringArm to ADS location; ADS does not use spring arm for positioning, it moves the camera seperate from the arm
      (CameraTimelineFunctions|UpdateDistanceTimeline|ADSEnabledUpdate (Variables|CameraDistance|GetCamera%20ADS%20Location) (LocalVariables|GetInputAlpha) (LocalVariables|GetInputTargetDistance)))
    (:NewEnumerator2
      ;[1 1 1] Move CameraAttach back to SpringArm 0,0,0
      (CameraTimelineFunctions|UpdateDistanceTimeline|ADSEnabledUpdate "0, 0, 0" (LocalVariables|GetInputAlpha) (LocalVariables|GetInputTargetDistance)))
    (:NewEnumerator3
      ;[1 1 1] Focus Point Zoom In
      (CameraTimelineFunctions|UpdateRotationTimeline|UpdateCameraFocusRotation (CameraTimelineFunctions|UpdateRotationTimeline|GetRotationtoFocusPointActor) (LocalVariables|GetInputAlpha)))
    (:NewEnumerator4
      ;[1 1 1] Focus Point Hold
      (Transformation|SetWorldRotation (Variables|Default|GetCameraAttach) (CameraTimelineFunctions|UpdateRotationTimeline|GetRotationtoFocusPointActor) false false))
    (:NewEnumerator5
      (CameraTimelineFunctions|UpdateRotationTimeline|ResetCameraRotation (LocalVariables|GetInputAlpha)))))

(fn ClampCameraPivotRotationY
  (params (Value :type Float%20%28double-precision%29))
  (bind _output_get (Variables|PivotRotation|SetCurrentCameraRotationY (Math|Float|Clamp(Float) (+ Value (Variables|PivotRotation|GetCurrentCameraRotationY)) (* (Variables|PivotRotation|GetMaxRotationY) -1.0) (* (Variables|PivotRotation|GetMaxRotationY) 0.75))))
  (return (Math|Float|InRange(Float) _output_get (* (Variables|PivotRotation|GetMaxRotationY) -1.0) (Variables|PivotRotation|GetMaxRotationY) false false)))

(fn ClampCameraPivotRotationZ
  (params (Value :type Float%20%28double-precision%29))
  (bind _output_get (Variables|PivotRotation|SetCurrentCameraRotationZ (Math|Float|Clamp(Float) (+ Value (Variables|PivotRotation|GetCurrentCameraRotationZ)) (* (Variables|PivotRotation|GetMaxRotationZ) -1.0) (Variables|PivotRotation|GetMaxRotationZ))))
  (return (Math|Float|InRange(Float) _output_get (* (Variables|PivotRotation|GetMaxRotationZ) -1.0) (Variables|PivotRotation|GetMaxRotationZ) false false)))

(fn DisableCameraModeADS
  (params (TimelineLengthSeconds :type Float%20%28double-precision%29 :default 1.0))
  (bind _focuspointresettime (Variables|FocusPoint|GetFocusPointResetTime))
  ;[1 0.1533 0] Disable ADS Zoom and rotate camera to default position if Focus Point was active
  (if (Variables|CameraDistance|GetCameraModeADS)
    (Variables|CameraDistance|SetCameraModeADS false)
    (CallFunction|StartCameraDistanceChangeTimeline TimelineLengthSeconds "NewEnumerator2" (Variables|CameraDistance|GetStandardModeArmLength))
    (if (Variables|FocusPoint|GetFocusPointActive)
      (CallFunction|RecenterCameraPivot _focuspointresettime)
      (CallFunction|StartCameraRotationChangeTimeline _focuspointresettime "NewEnumerator5"))))

(fn DisableFocusPoint
  ;[1 0.1239 0] Set Focus Point Inactive
  (Variables|FocusPoint|SetFocusPointActive false))

(fn EnableFocusPoint
  (params (Time%20To%20Focus :type Float%20%28double-precision%29 :default 1.0) (Focus%20Duration :type Float%20%28double-precision%29 :default 1.0) (ActorForFocus :type Actor%20Object%20Reference))
  ;[0.2834 1 0] Sets Focus Point timer and toggles boolean ON. Will rotate camera if ADS Mode is active.
  (Utilities|Time|ClearandInvalidateTimerbyHandle (Variables|FocusPoint|GetFocusPointTimer))
  (Variables|FocusPoint|SetActorForFocus ActorForFocus)
  (Variables|FocusPoint|SetTime%20to%20Focus Time%20To%20Focus)
  (bind _returnvalue (Utilities|Time|SetTimerbyEvent (EventDispatchers|CreateEvent :!event DisableFocusPoint) Focus%20Duration))
  (Variables|FocusPoint|SetFocusPointTimer _returnvalue)
  (Variables|FocusPoint|SetFocusPointActive true)
  (if (Variables|CameraDistance|GetCameraModeADS)
    (FocusPoint|ActivateFocusPoint)))

(fn GetRotationToFocusPointActor
  (bind _actorforfocus (Variables|FocusPoint|GetActorForFocus)
    (:then
      (return (Math|Rotator|FindLookatRotation (Transformation|GetWorldLocation (Variables|Default|GetCameraAttach)) (Transformation|GetActorLocation _actorforfocus))))
    (:else
      (Development|PrintString "File=//Fortnite/Main/FortniteGame/Content/Athena/Coneric/BP_Generic_SpringArm.uasset : Invalid ActorForFocus [BP_Generic_SpringArm->CameraTimelineUpdateManager->GetRotationToFocusPointActor]")
      (return "0, 0, 0"))))

(fn RecenterCameraPivotRotation
  (params (Alpha :type Float%20%28double-precision%29))
  (bind _pivot (Variables|Default|GetPivot))
  (bind (_relativerotation_roll _relativerotation_pitch _relativerotation_yaw) (Math|Rotator|BreakRotator (Variables|/Script/Engine.SceneComponent|GetRelativeRotation_Roll _pivot)))
  (Transformation|SetRelativeRotation (Variables|Default|GetPivot) (Math|Rotator|Lerp(Rotator) (Math|Rotator|MakeRotator 0.0 (Variables|PivotRotation|GetCurrentCameraRotationY) (Variables|PivotRotation|GetCurrentCameraRotationZ)) "0, 0, 0" (* Alpha (Variables|PivotRotation|GetRecenterRotationSpeed)) true) false false)
  (Variables|PivotRotation|SetCurrentCameraRotationZ _relativerotation_yaw)
  (Variables|PivotRotation|SetCurrentCameraRotationY _relativerotation_pitch))

(fn ResetCameraRotation
  (params (Alpha :type Float%20%28double-precision%29))
  (Transformation|SetRelativeRotation (Variables|Default|GetCameraAttach) (Math|Rotator|Lerp(Rotator) (Variables|/Script/Engine.SceneComponent|GetRelativeRotation (Variables|Default|GetCameraAttach)) "0, 0, 0" Alpha true) false false))

(fn SetArmLengthPostChange
  (params (SetNewStandardModeLength :type Boolean))
  (Utilities|FlowControl|Sequence
    (:then_0
      (bind _output_get (Variables|CameraDistance|SetSavedArmLength (Variables|/Script/Engine.SpringArmComponent|GetTargetArmLength (Variables|Default|GetSpringArm)))))
    (:then_1
      (if SetNewStandardModeLength
        (Variables|CameraDistance|SetStandardModeArmLength _output_get)))))

(fn SetCameraDistance
  (params (NewCameraDistance :type Float%20%28double-precision%29) (TimelineLengthSeconds :type Float%20%28double-precision%29 :default 1.0))
  ;[0.2618 0 1] Handle Updating Camera Distance
  (if (Variables|CameraDistance|GetCameraModeADS)
    (Variables|CameraDistance|SetStandardModeArmLength NewCameraDistance)
    (else
      (CallFunction|StartCameraDistanceChangeTimeline TimelineLengthSeconds "NewEnumerator0" NewCameraDistance))))

(fn SetCameraModeADS
  (params (ADS%20Camera%20Distance :type Float%20%28double-precision%29) (Timeline%20Length%20Seconds :type Float%20%28double-precision%29 :default 1.0))
  ;[0.3445 1 0] ADS Zoom to bring camera tighter to character. Will rotate to focus point if active.
  (Variables|CameraDistance|SetCameraModeADS true)
  (CallFunction|StartCameraDistanceChangeTimeline Timeline%20Length%20Seconds "NewEnumerator1" ADS%20Camera%20Distance)
  (if (Variables|FocusPoint|GetFocusPointActive)
    (CallFunction|RecenterCameraPivot 0.0)
    (FocusPoint|ActivateFocusPoint)))

(fn SmoothRotation
  (params (CurrentRotation :type Float%20%28double-precision%29) (TargetRotation :type Float%20%28double-precision%29))
  (return (Math|Interpolation|FInterpTo CurrentRotation (+ CurrentRotation TargetRotation) (Utilities|Time|GetWorldDeltaSeconds) (Utilities|Time|GetWorldDeltaSeconds))))

(fn UpdateMaxRotationsAndSpeed
  (params (MaxRotationZ :type Float%20%28double-precision%29) (MaxRotationY :type Float%20%28double-precision%29) (CameraSpeed :type Float%20%28double-precision%29))
  (Utilities|FlowControl|Sequence
    (:then_0
      (if (> MaxRotationZ 0)
        (Variables|PivotRotation|SetMaxRotationZ MaxRotationZ)))
    (:then_1
      (if (> MaxRotationY 0)
        (Variables|PivotRotation|SetMaxRotationY MaxRotationY)))
    (:then_2
      (if (> CameraSpeed 0)
        (Variables|PivotRotation|SetCameraSpeed CameraSpeed)))))

(fn UpdateCameraDistance
  (params (Alpha :type Float%20%28double-precision%29) (NewCameraDistance :type Float%20%28double-precision%29 :default 5.5e+02))
  (Variables|/Script/Engine.SpringArmComponent|SetTargetArmLength (Math|Float|Lerp (Variables|CameraDistance|GetSavedArmLength) NewCameraDistance Alpha) (Variables|Default|GetSpringArm)))

(fn UpdateCameraFocusRotation
  (params (TargetRotation :type Rotator) (Alpha :type Float%20%28double-precision%29))
  (Transformation|SetWorldRotation (Variables|Default|GetCameraAttach) (Math|Rotator|Lerp(Rotator) (Transformation|GetWorldRotation (Variables|Default|GetCameraAttach)) TargetRotation Alpha true) false false))

(fn ConstructionScript)
